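# One-shot Niagara effect for the landing in the VileSting emote: a red-orange impact burst with pale airflow swirls.
# A system socket reader places the energy wave, ring, glows with light, sparks and squares; airflow meshes use fixed offsets.
# Notes_Line bends five ribbons into camera-facing expanding rings with curve-driven spikes resembling an audio waveform.
# Time-of-day functions tint or dim several layers, and front-end modifiers rescale some for preview.
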NiagaraSystem /BRCosmetics/Animation/Game/MainPlayer/Emotes/VileSting/FX/NS_Emote_VileSting_Landing
dependencies {
 usedDataInterfaces [
  &NiagaraDataInterfaceArrayColor
  &NiagaraDataInterfaceArrayFloat
  &NiagaraDataInterfaceCurve
  &NiagaraDataInterfaceVectorCurve
  &NiagaraDataInterfaceColorCurve
  &NiagaraDataInterfaceCamera
  &NiagaraDataInterfaceSpriteRendererInfo
  &NiagaraDataInterfaceSocketReader
 ]
 usedDynamicInputs [
  &RotateVectorByQuaternion
  &NormalizeVector
  &MakeVector2D
  &SimulationPosition
  &UniformRangedVector
 ]
 usedModules [
  &self:SystemSpawnScript.SystemScriptSource.SystemScriptGraph.NiagaraNodeAssignment_0.SetVariables_#1
  &self:SystemSpawnScript.SystemScriptSource.SystemScriptGraph.NiagaraNodeAssignment_1.SetVariables_#2
  &SystemState
 ]
}
emitters {
 Airflow_Radial {
  emitterUpdateScript [
   &EmitterState()
   &SpawnBurst_Instantaneous(Spawn Count=3, Use Spawn Probability=false, UseLoopCountLimit=false)
  ]
  particleSpawnScript [
   &InitializeParticle(
    Color=<rgba 0.8542 0.9725 1 1>
    Lifetime Max=1
    Lifetime Min=0.8
    Lifetime Mode=::NewEnumerator1
    Mesh Scale Mode=::NewEnumerator3
    Mesh Uniform Scale=2
    Position Offset=<0 -30 0>
    UsePositionOffset=true
    Write Material Random=true
   )
   &InitialMeshOrientation(
    Mesh Orientation Mode=::NewEnumerator7
    Rotation=&RandomRangeVector(Minimum=<0 0 0>, Maximum=<0 0 1>)
    Use Rotation Vector=true
   )
  ]
  particleUpdateScript [
   &ParticleState()
   &DynamicMaterialParameters(
    Index 0 Param 1=&FloatFromCurve(FloatCurve=DataInterface() {propertyValues {Curve {keys [{interpMode=RCIM_Cubic, tangentMode=RCTM_User, time=0.5, value=0}, {interpMode=RCIM_Cubic, tangentMode=RCTM_User, time=1, value=1}]}}})
    Index 0 Param 2=1
    Index 0 Param 3=1
    Module.Param0WriteEnabled=true
    Module.Param1WriteEnabled=true
    Module.Param2WriteEnabled=true
    Module.Param3WriteEnabled=true
   )
   &ScaleMeshSize(Scale Factor=&VectorFromFloat(Value=&FloatFromCurve(FloatCurve=DataInterface() {propertyValues {Curve {keys [
          {arriveTangent=1.2501, interpMode=RCIM_Cubic, leaveTangent=1.2501, tangentMode=RCTM_User, time=0, value=0.3}
          {interpMode=RCIM_Cubic, tangentMode=RCTM_User, time=1, value=1}
         ]}}})))
   &UpdateMeshOrientation(
    Rotation Rate=&FloatFromCurve(
     FloatCurve=DataInterface() {propertyValues {Curve {
        defaultValue=1
        keys [
         {arriveTangent=-1.7715, interpMode=RCIM_Cubic, leaveTangent=-1.7715, tangentMode=RCTM_User, time=0, value=1}
         {interpMode=RCIM_Cubic, tangentMode=RCTM_User, time=1, value=0}
        ]
       }}}
     Scale Curve=-1.5
    )
    Rotation Vector=<0 0 1>
   )
   &ScaleColor(
    Scale Alpha=&FloatFromCurve(
     FloatCurve=DataInterface() {propertyValues {Curve {keys [
         {arriveTangent=8.2356, interpMode=RCIM_Cubic, leaveTangent=8.2356, tangentMode=RCTM_User, time=0, value=0}
         {arriveTangent=0.1969, interpMode=RCIM_Cubic, leaveTangent=0.1969, tangentMode=RCTM_User, time=0.2223, value=1}
         {arriveTangent=-1.2858, interpMode=RCIM_Cubic, leaveTangent=-1.2858, time=0.5, value=0.6}
         {arriveTangent=-0.7636, interpMode=RCIM_Cubic, leaveTangent=-0.7636, tangentMode=RCTM_User, time=1, value=0}
        ]}}}
     Scale Curve=0.5
    )
    Scale RGB=&Multiply_VectorByFloat(
     Vector=&MakeVectorFromLinearColorRGB(Color=&FNTimeOfDayColor(
       Night Color=<rgba 0.474 0.6858 1 1>
       Morning Color=<rgba 0.75 0.8507 1 1>
       Day Color=<rgba 0.75 0.8507 1 1>
       Evening Color=<rgba 0.75 0.8507 1 1>
      ))
     Float=&FloatFromCurve(
      FloatCurve=DataInterface() {propertyValues {Curve {keys [
          {arriveTangent=6.747, interpMode=RCIM_Cubic, leaveTangent=6.747, tangentMode=RCTM_User, time=0, value=0}
          <at 0.2 1 RCIM_Cubic>
          {arriveTangent=-0.6427, interpMode=RCIM_Cubic, leaveTangent=-0.6427, tangentMode=RCTM_User, time=1, value=0}
         ]}}}
      Scale Curve=&Multiply_Float(A=1, B=&FNTimeOfDayScalar(Night Scale=0.8, Morning Scale=0.8, Evening Scale=0.8))
     )
    )
    ScaleA=true
    ScaleRGB=true
   )
  ]
  rendererClasses [&NiagaraMeshRendererProperties]
  renderers [{rendererClass=&NiagaraMeshRendererProperties, rendererIndex=0}]
 }
 Airflow_Spline {
  emitterUpdateScript [
   &EmitterState()
   &SpawnRate(
    SpawnRate=&FloatFromCurve(
     FloatCurve=DataInterface() {propertyValues {Curve {
        defaultValue=1
        keys [
         {arriveTangent=-2.5624, interpMode=RCIM_Cubic, leaveTangent=-2.5624, tangentMode=RCTM_User, time=0, value=1}
         {interpMode=RCIM_Cubic, tangentMode=RCTM_User, time=0.5, value=0}
        ]
       }}}
     CurveIndex=$Emitter.NormalizedLoopAge
     Scale Curve=10
    )
    Use Spawn Probability=false
   )
   &SpawnBurst_Instantaneous(Spawn Count=5, Use Spawn Probability=false, UseLoopCountLimit=false)
  ]
  particleSpawnScript [
   &InitializeParticle(
    Lifetime=0.6
    Mesh Scale=<1 1 1.5>
    Mesh Scale Mode=::NewEnumerator4
    Position Offset=<0 -30 10>
    UsePositionOffset=true
    Write Material Random=true
   )
   &InitialMeshOrientation(
    Mesh Orientation Mode=::NewEnumerator2
    Rotation=&RandomRangeVector(Minimum=<0 0 0>, Maximum=<0 0 1>)
    Use Rotation Vector=true
   )
  ]
  particleUpdateScript [
   &ParticleState()
   &UpdateMeshOrientation(
    Rotation Rate=&FloatFromCurve(
     FloatCurve=DataInterface() {propertyValues {Curve {
        defaultValue=0
        keys [
         {arriveTangent=-2.0483, interpMode=RCIM_Cubic, leaveTangent=-2.0483, tangentMode=RCTM_User, time=0, value=1}
         {interpMode=RCIM_Cubic, tangentMode=RCTM_User, time=1, value=0}
        ]
       }}}
     Scale Curve=-1
    )
    Rotation Vector=<0 0 1>
   )
   &ScaleMeshSize(Scale Factor=&Multiply_VectorByFloat(
     Vector=&VectorFromCurve(VectorCurve=DataInterface() {propertyValues {
        XCurve {
         defaultValue=0.6
         keys [
          {arriveTangent=1.2805, interpMode=RCIM_Cubic, leaveTangent=1.2805, tangentMode=RCTM_User, time=0, value=0.2}
          {interpMode=RCIM_Cubic, tangentMode=RCTM_User, time=1, value=1.2}
         ]
        }
        YCurve {
         defaultValue=0.6
         keys [
          {arriveTangent=1.2805, interpMode=RCIM_Cubic, leaveTangent=1.2805, tangentMode=RCTM_User, time=0, value=0.2}
          {interpMode=RCIM_Cubic, tangentMode=RCTM_User, time=1, value=1.2}
         ]
        }
        ZCurve {
         defaultValue=0.6
         keys [{interpMode=RCIM_Cubic, tangentMode=RCTM_User, time=0, value=0.5}, {interpMode=RCIM_Cubic, tangentMode=RCTM_User, time=1, value=1.2}]
        }
       }})
     Float=1
    ))
   &DynamicMaterialParameters(
    Index 0 Param 1=&FloatFromCurve(FloatCurve=DataInterface() {propertyValues {Curve {keys [
         {arriveTangent=-0.3441, interpMode=RCIM_Cubic, leaveTangent=-0.3441, tangentMode=RCTM_User, time=0, value=0.3}
         {arriveTangent=0.0064, interpMode=RCIM_Cubic, leaveTangent=0.0064, tangentMode=RCTM_User, time=1, value=0.1}
        ]}}})
    Index 0 Param 2=&RandomRangeFloat(Minimum=60, Maximum=70)
    Module.Index0Param0Write=true
    Module.Index0Param1Write=true
    Module.Index0Param2Write=true
    Module.Index0Param3Write=true
   )
   &ScaleColor(
    Scale Alpha=&FloatFromCurve(
     FloatCurve=DataInterface() {propertyValues {Curve {
        defaultValue=1
        keys [
         {arriveTangent=9.0725, interpMode=RCIM_Cubic, leaveTangent=9.0725, tangentMode=RCTM_User, time=0, value=0}
         {arriveTangent=0.0231, interpMode=RCIM_Cubic, leaveTangent=0.0231, tangentMode=RCTM_User, time=0.1408, value=1}
         {arriveTangent=-0.0405, interpMode=RCIM_Cubic, leaveTangent=-0.0405, tangentMode=RCTM_User, time=1, value=0}
        ]
       }}}
     Scale Curve=&Multiply_Float(A=0.5, B=&FNTimeOfDayScalar(Night Scale=0.8, Morning Scale=0.8, Evening Scale=0.8))
    )
    Scale RGB=&MakeVectorFromLinearColorRGB(Color=<rgba 1 1 1 1>)
    ScaleA=true
    ScaleRGB=true
   )
  ]
  rendererClasses [&NiagaraMeshRendererProperties]
  renderers [{rendererClass=&NiagaraMeshRendererProperties, rendererIndex=0}]
 }
 Block_Ring {
  emitterUpdateScript [
   &EmitterState()
   &SpawnBurst_Instantaneous(Use Spawn Probability=false, UseLoopCountLimit=false)
  ]
  particleSpawnScript [
   &InitializeParticle(
    Lifetime=0.5
    Sprite Rotation Mode=::NewEnumerator2
    Sprite Size Mode=::NewEnumerator3
    Uniform Sprite Size=220
    UsePositionOffset=false
    Write Material Random=true
   )
   &CameraOffset(Camera Offset Amount=-10)
   &SocketLocation(
    Orientation Sampling=::NewEnumerator1
    SamplePosOffset=true
    Socket Reader=$System.Socket Reader
    Socket Sampling Mode=::NewEnumerator3
    VelocitySpeedLimit=false
   )
  ]
  particleUpdateScript [
   &ParticleState()
   &ScaleSpriteSize(
    Uniform Curve Scale=&MakeCustomFloatFromBool(Bool=$User.bIsFrontEndPreview, True Float=0.7, False Float=1)
    Uniform Curve Sprite Scale=DataInterface() {propertyValues {Curve {keys [
        {arriveTangent=4.4465, interpMode=RCIM_Cubic, leaveTangent=4.4465, tangentMode=RCTM_User, time=0, value=0.5}
        {arriveTangent=1.5, interpMode=RCIM_Cubic, leaveTangent=1.5, time=0.3, value=1.5}
        <at 1 2 RCIM_Cubic>
       ]}}}
   )
   &SubUVAnimation(
    End Frame Range Override=3
    Random Start Frame=true
    "Sprite Renderer "=DataInterface() {propertyValues {SpriteRenderer=&self:Block_Ring_0.NiagaraSpriteRendererProperties_0}}
    SubUV Lookup Index Scale=0.8
    UseEndFrame=true
    UseStartFrame=true
   )
   &ScaleColor(
    Scale Alpha=&FloatFromCurve(
     FloatCurve=DataInterface() {propertyValues {Curve {
        defaultValue=0.8034
        keys [{interpMode=RCIM_Cubic, tangentMode=RCTM_User, time=0, value=1}, {interpMode=RCIM_Cubic, tangentMode=RCTM_User, time=1, value=0}]
       }}}
     Scale Curve=1.3
    )
    Scale RGB=&MakeVectorFromLinearColorRGB(Color=&ColorFromCurve(ColorCurve=DataInterface() {propertyValues {
        AlphaCurve {keys [<at 0 1 RCIM_Cubic>]}
        BlueCurve {keys [<at 0 0 RCIM_Cubic>,<at 1 0 RCIM_Cubic>]}
        GreenCurve {keys [<at 0 0.15 RCIM_Cubic>,<at 1 0 RCIM_Cubic>]}
        RedCurve {keys [<at 0 3 RCIM_Cubic>,<at 1 1 RCIM_Cubic>]}
       }}))
    ScaleA=true
    ScaleRGB=true
   )
  ]
  rendererClasses [&NiagaraSpriteRendererProperties]
  renderers [{rendererClass=&NiagaraSpriteRendererProperties}]
 }
 EnergyWave {
  emitterUpdateScript [
   &EmitterState()
   &SpawnBurst_Instantaneous(Spawn Count=1, Use Spawn Probability=false)
  ]
  particleSpawnScript [
   &InitializeParticle(Color=<rgba 0.4323 0.0228 0.0086 1>, Lifetime=0.8, Mesh Scale Mode=::NewEnumerator3, UsePositionOffset=false, Write Material Random=true) {module=InitializeParticle001}
   &InitialMeshOrientation(Orientation Axis=<0 0 1>, Orientation Vector=<0 0 1>, Rotation=<0 0.25 0>, Use Orientation Vector=true, Use Rotation Vector=true)
   &InitialMeshOrientation(
    Orientation Axis=<0 0 0>
    Orientation Vector=<0 0 0>
    Rotation=&RandomRangeVector(Minimum=<0 0 0>, Maximum=<1 0 0>)
    Use Orientation Vector=false
    Use Rotation Vector=true
   ) {module=InitialMeshOrientation001}
   &SocketLocation(
    Orientation Sampling=::NewEnumerator1
    SamplePosOffset=true
    Socket Reader=$System.Socket Reader
    Socket Sampling Mode=::NewEnumerator3
    VelocitySpeedLimit=false
   )
  ]
  particleUpdateScript [
   &DynamicMaterialParameters(
    Index 0 Param 1=&FloatFromCurve(FloatCurve=DataInterface() {propertyValues {Curve {keys [
         {arriveTangent=-5.6453, interpMode=RCIM_Cubic, leaveTangent=-5.6453, tangentMode=RCTM_User, time=0, value=0.6}
         {arriveTangent=-0.6085, interpMode=RCIM_Cubic, leaveTangent=-0.6085, tangentMode=RCTM_User, time=0.2, value=0}
         <at 1 -0.25>
        ]}}})
    Index 0 Param 2=&FloatFromCurve(FloatCurve=DataInterface() {propertyValues {Curve {keys [{arriveTangent=-0.3464, interpMode=RCIM_Cubic, leaveTangent=-0.3464, tangentMode=RCTM_User, time=0, value=1}, <at 1 0.9>]}}})
    Index 0 Param 3=&FloatFromCurve(FloatCurve=DataInterface() {propertyValues {Curve {keys [<at 0 0>,<at 0.15 1>,<at 1 1.2>]}}})
    Module.Param0WriteEnabled=true
    Module.Param1WriteEnabled=true
    Module.Param2WriteEnabled=true
    Module.Param3WriteEnabled=false
   )
   &ScaleColor(
    Scale Alpha=&FloatFromCurve(FloatCurve=DataInterface() {propertyValues {Curve {keys [<at 0 1>,<at 1 0>]}}})
    Scale RGB=&VectorFromFloat(Value=&FloatFromCurve(FloatCurve=DataInterface() {propertyValues {Curve {keys [<at 0 10>,<at 0.15 3>,<at 1 1>]}}}))
    ScaleA=true
    ScaleRGB=true
   ) {module=ScaleColor001}
   &ParticleState()
   &ScaleMeshSize(Scale Factor=&VectorFromFloat(Value=&FloatFromCurve(FloatCurve=DataInterface() {propertyValues {Curve {keys [
          {arriveTangent=17.6756, interpMode=RCIM_Cubic, leaveTangent=17.6756, tangentMode=RCTM_User, time=0, value=0}
          {arriveTangent=1.4317, interpMode=RCIM_Cubic, leaveTangent=1.4317, tangentMode=RCTM_User, time=0.1, value=1.4}
          <at 1 2 RCIM_Cubic>
         ]}}})))
  ]
  rendererClasses [&NiagaraMeshRendererProperties]
  renderers [{rendererClass=&NiagaraMeshRendererProperties, rendererIndex=0}]
 }
 Glows {
  emitterUpdateScript [
   &EmitterState()
   &SpawnBurst_Instantaneous(Spawn Count=1, Use Spawn Probability=false)
  ]
  particleSpawnScript [
   &InitializeParticle(
    Color=<rgba 0.01 0.0083 0.0077 1.5>
    Lifetime=0.5
    Mass Mode=::NewEnumerator1
    Sprite Rotation Mode=::NewEnumerator2
    Sprite Size Mode=::NewEnumerator3
    Uniform Sprite Size=250
    UsePositionOffset=false
    Write Material Random=true
   ) {module=InitializeParticle001}
   &SocketLocation(
    Orientation Sampling=::NewEnumerator1
    SamplePosOffset=true
    Socket Reader=$System.Socket Reader
    Socket Sampling Mode=::NewEnumerator3
    VelocitySpeedLimit=false
   )
   &CameraOffset(Camera Offset Amount=-10)
  ]
  particleUpdateScript [
   &ParticleState()
   &DynamicMaterialParameters(
    Index 0 Param 1=&FloatFromCurve(FloatCurve=DataInterface() {propertyValues {Curve {keys [<at 0 1 RCIM_Cubic>,<at 0.8 0 RCIM_Cubic>]}}})
    Index 0 Param 2=2
    Index 0 Param 3=1
    Index 0 Param 4=1
    Module.Index0Float0Write=true
    Module.Index0Float1Write=true
    Module.Index0Float2Write=true
    Module.Index0Float3Write=true
   )
   &ScaleColor(
    Scale Alpha=&FloatFromCurve(
     FloatCurve=DataInterface() {propertyValues {Curve {
        defaultValue=1
        keys [
         {arriveTangent=-1.3746, interpMode=RCIM_Cubic, leaveTangent=-1.3746, tangentMode=RCTM_User, time=0, value=1}
         {interpMode=RCIM_Cubic, tangentMode=RCTM_User, time=1, value=0}
        ]
       }}}
     Scale Curve=&FNTimeOfDayScalar(Night Scale=0.5, Morning Scale=1, Evening Scale=1)
    )
    ScaleA=true
    ScaleRGB=true
   )
   &FN_Sprite_Attributes(
    Color=&MultiplyLinearColor(
     Linear Color=<rgba 2 0.1667 0 1>
     Scale Factor=&MakeVector4FromVectorAndFloat(
      "Vector (RGB)"=<1 1 1>
      "Float (Alpha)"=&FloatFromCurve(FloatCurve=DataInterface() {propertyValues {Curve {
          defaultValue=1
          keys [
           {arriveTangent=-1.3503, interpMode=RCIM_Cubic, leaveTangent=-1.3503, tangentMode=RCTM_User, time=0, value=1}
           {interpMode=RCIM_Cubic, tangentMode=RCTM_User, time=1, value=0}
          ]
         }}})
     )
    )
    Create Color Attribute=true
    Create SpriteSize Attribute=true
    SpriteSize=&Multiply_Vector2D_ByFloat(Float=0.5, Vector2D=$Particles.SpriteSize)
    Write to Module Namespace=true
   )
   &FN_Light_Attributes(
    Color=<rgba 1 0 0 1>
    Create Color Attribute=true
    Create Radius Attribute=true
    Radius=100
    Scale Color=&FloatFromCurve(
     FloatCurve=DataInterface() {propertyValues {Curve {
        defaultValue=1
        keys [
         {arriveTangent=2.7941, interpMode=RCIM_Cubic, leaveTangent=2.7941, tangentMode=RCTM_User, time=0, value=0.7}
         {arriveTangent=-1.2912, interpMode=RCIM_Cubic, leaveTangent=-1.2912, tangentMode=RCTM_User, time=1, value=0}
        ]
       }}}
     Scale Curve=8
    )
   )
  ]
  rendererClasses [&NiagaraSpriteRendererProperties, &NiagaraLightRendererProperties]
  renderers [
   {rendererClass=&NiagaraSpriteRendererProperties}
   {rendererClass=&NiagaraSpriteRendererProperties, rendererIndex=1}
   {rendererClass=&NiagaraLightRendererProperties, rendererIndex=2}
  ]
 }
 Notes_Line {
  emitterSpawnScript [/SetVariables(Emitter.Rand:NiagaraFloat=&RandomRangeFloat(Minimum=0, Maximum=1)) {isSetParameters=true}]
  emitterUpdateScript [
   &EmitterState()
   &SpawnBurst_Instantaneous(Spawn Count=120, Use Spawn Probability=false, UseLoopCountLimit=false)
  ]
  particleSpawnScript [
   &PartitionParticles(
    Normalized Index In Partition Includes One=true
    Number Of Partitions=5
    Particles Per Partition=24
    Partition Mode:ENiagara_PartitionMode=::NewEnumerator1
    Write Partition to Ribbon ID=::NewEnumerator0
   )
   &InitializeParticle(
    Color Mode=::NewEnumerator0
    Lifetime=0.5
    Ribbon Width=10
    Ribbon Width Mode=::NewEnumerator1
    UsePositionOffset=false
    Write Material Random=true
   )
   /SetVariables(Particles.RibbonLinkOrder:NiagaraFloat=$Particles.PartitionParticles.Normalized Index In Partition) {isSetParameters=true}
  ]
  particleUpdateScript [
   &ParticleState()
   &Color(Color=&SelectLinearColorFromArray(
     Array Sampling Mode=::NewEnumerator1
     Direct Array Index=$Particles.PartitionParticles.Partition
     Color Selection Array=DataInterface() {propertyValues {ColorData [<rgba 2 0.1 0 1>, <rgba 1.5 0.05 0 1>, <rgba 1.25 0 0 1>, <rgba 0.5 0 0 1>, <rgba 0.2 0 0 1>]}}
     Random Seed=?
    ))
   &ScaleColor(
    Color Value To Scale=$Particles.Color
    Scale Alpha=&FloatFromCurve(
     FloatCurve=DataInterface() {propertyValues {Curve {keys [
         {arriveTangent=-2.3502, interpMode=RCIM_Cubic, leaveTangent=-2.3502, tangentMode=RCTM_User, time=0, value=1}
         {interpMode=RCIM_Cubic, tangentMode=RCTM_User, time=1, value=0}
        ]}}}
     Scale Curve=0.5
    )
    Scale RGB=&MakeVectorFromLinearColorRGB(Color=&FNTimeOfDayColor(
      Night Color=<rgba 0.5625 0.6225 0.75 1>
      Morning Color=<rgba 0.7812 0.5615 0.5615 1>
      Day Color=<rgba 1 1 1 1>
      Evening Color=<rgba 0.7812 0.5615 0.5615 1>
     ))
    ScaleA=true
    ScaleRGB=true
   )
   &SolveForcesAndVelocity(Clamp Velocity=false, Limit Acceleration=false)
   &SocketLocation(
    Orientation Sampling=::NewEnumerator1
    SamplePosOffset=true
    Socket Reader=$System.Socket Reader
    Socket Sampling Mode=::NewEnumerator3
    VelocitySpeedLimit=false
   )
   &RotateAroundPoint(
    Radius:NiagaraFloat=&FloatFromCurve(
     FloatCurve=DataInterface() {propertyValues {Curve {keys [
         {arriveTangent=1.3882, interpMode=RCIM_Cubic, leaveTangent=1.3882, tangentMode=RCTM_User, time=0, value=0.3}
         {arriveTangent=0.0922, interpMode=RCIM_Cubic, leaveTangent=0.0922, tangentMode=RCTM_User, time=1, value=1}
        ]}}}
     Scale Curve=&SelectFloatFromArray(
      Direct Array Index=$Particles.PartitionParticles.Partition
      Float Selection Array=DataInterface() {propertyValues {FloatData [80,90,100,110,120]}}
     )
    )
    Rotation Center:NiagaraPosition=$Particles.Position
    Rotation Phase:NiagaraFloat=&Add_Float(A=$Emitter.Rand, B=$Particles.PartitionParticles.Normalized Index In Partition)
    X Axis:Vector3f=&ReturnCameraVector(Camera Vectors=::NewEnumerator2, Coordinate Space=::Simulation)
    Y Axis:Vector3f=&ReturnCameraVector(Camera Vectors=::NewEnumerator3, Coordinate Space=::Simulation)
   )
   /SetVariables(Particles.Position=&AddVectorToPosition(
     Position=$Particles.Position
     Vector=&Multiply_VectorByFloat(
      Vector=$Particles.RotateAroundPoint.VectorFromCenter
      Float=&FloatFromCurve(
       FloatCurve=DataInterface() {propertyValues {Curve {keys [
           {arriveTangent=9.0725, interpMode=RCIM_Cubic, leaveTangent=9.0725, tangentMode=RCTM_User, time=0, value=0.25}
           {arriveTangent=-2.2349, interpMode=RCIM_Cubic, leaveTangent=-2.2349, tangentMode=RCTM_User, time=0.2, value=1}
           {arriveTangent=-0.0208, interpMode=RCIM_Cubic, leaveTangent=-0.0208, tangentMode=RCTM_User, time=1, value=0}
          ]}}}
       Scale Curve=&Multiply_Float(
        A=60
        B=&FloatFromCurve(
         FloatCurve=DataInterface() {propertyValues {Curve {keys [
             <at 0 0.8>
             {arriveTangent=-6.0547, time=0.025, value=0}
             <at 0.08 0>
             <at 0.1 0.5>
             <at 0.12 0>
             {arriveTangent=5.533, leaveTangent=5.533, time=0.16, value=0}
             <at 0.202 0.75>
             {arriveTangent=-7.2386, leaveTangent=-7.2386, time=0.24, value=0}
             <at 0.28 0>
             <at 0.3 0.6>
             <at 0.32 0>
             {arriveTangent=3.8652, leaveTangent=3.8652, time=0.3523, value=0}
             <at 0.3723 0.5>
             {arriveTangent=-3.7702, leaveTangent=-3.7702, time=0.3923, value=0}
             {arriveTangent=8.0575, leaveTangent=8.0575, time=0.48, value=0}
             <at 0.5 0.8>
             {arriveTangent=-10.12, leaveTangent=-10.12, time=0.52, value=0}
             {arriveTangent=7.1912, leaveTangent=7.1912, time=0.6, value=0}
             <at 0.62 0.6>
             {arriveTangent=-8.6269, leaveTangent=-8.6269, time=0.64, value=0}
             {arriveTangent=15.9213, leaveTangent=15.9213, time=0.68, value=0}
             <at 0.7 0.5>
             {arriveTangent=-8.1956, leaveTangent=-8.1956, time=0.72, value=0}
             {arriveTangent=7.7748, leaveTangent=7.7748, time=0.84, value=0}
             {arriveTangent=-9.6894, leaveTangent=-9.6894, time=0.85, value=0.5}
             {arriveTangent=-7.785, leaveTangent=-7.785, time=0.86, value=0}
             {arriveTangent=9.6548, leaveTangent=9.6548, time=0.975, value=0}
             <at 1 0.8>
            ]}}}
         CurveIndex=$Particles.PartitionParticles.Normalized Index In Partition
        )
       )
      )
     )
    )) {isSetParameters=true}
  ]
  rendererClasses [&NiagaraRibbonRendererProperties]
  renderers [{rendererClass=&NiagaraRibbonRendererProperties}]
 }
 Sparks {
  emitterUpdateScript [
   &EmitterState()
   &SpawnBurst_Instantaneous(Spawn Count=&UniformRangedInt(Minimum=26, Maximum=33), Use Spawn Probability=false)
  ]
  particleSpawnScript [
   &InitializeParticle(
    Lifetime Max=0.5
    Lifetime Min=0.25
    Lifetime Mode=::NewEnumerator1
    Sprite Size Max=<1.5 15>
    Sprite Size Min=<1 10>
    Sprite Size Mode=::NewEnumerator2
    UsePositionOffset=false
    Write Material Random=true
   )
   &CameraOffset(Camera Offset Amount=&FN_FrontEndFloatModifier(Gameplay=10, Front End=$System.FrontendCameraOffset))
   &SocketLocation(
    Interpolate Socket Sampling=true
    SamplePosOffset=true
    Socket Reader=$System.Socket Reader
    Socket Sampling Mode=::NewEnumerator3
    VelocitySpeedLimit=false
   )
   &ShapeLocation(
    MaskSpawnGroup=false
    Override Seed=false
    Sphere Radius=&Multiply_Float(
     A=&RandomRangeFloat(Minimum=30, Maximum=50)
     B=&FN_FrontEndFloatModifier(Gameplay=1, Front End=$System.FrontendScale)
    )
    Sphere Surface Distribution=1
    UseOwnerScale=true
   )
   &AddVelocity(
    Module.Use Falloff Distance=false
    Override Seed=false
    Velocity Mode=::NewEnumerator1
    Velocity Origin=$Output.SocketLocation.SampledPosition
    Velocity Speed=&Multiply_Float(
     A=&RandomRangeFloat(Minimum=111, Maximum=444)
     B=&FN_FrontEndFloatModifier(Gameplay=1, Front End=$System.FrontendScale)
    )
   )
  ]
  particleUpdateScript [
   &ParticleState()
   &ScaleColor(
    Scale Alpha=&FloatFromCurve(FloatCurve=DataInterface() {propertyValues {Curve {
        defaultValue=1
        keys [{interpMode=RCIM_Cubic, tangentMode=RCTM_User, time=0, value=1}, {interpMode=RCIM_Cubic, tangentMode=RCTM_User, time=1, value=0}]
       }}})
    Scale RGB=&MakeVectorFromLinearColorRGB(Color=&ColorFromCurve(ColorCurve=DataInterface() {propertyValues {
        AlphaCurve {keys [<at 0 1>,<at 1 0>]}
        BlueCurve {keys [<at 0.1007 0.15>,<at 1 0>]}
        GreenCurve {keys [<at 0.1007 0.15>,<at 1 0.0167>]}
        RedCurve {keys [<at 0.1007 3>,<at 1 1>]}
       }}))
    ScaleA=true
    ScaleRGB=true
   )
   &ScaleSpriteSize(Scale Factor=&Vector2DFromFloat(Value=&FloatFromCurve(
      FloatCurve=DataInterface() {propertyValues {Curve {keys [
          {arriveTangent=10.1867, interpMode=RCIM_Cubic, leaveTangent=10.1867, tangentMode=RCTM_User, time=0, value=0}
          {interpMode=RCIM_Cubic, leaveTangent=-0.0263, tangentMode=RCTM_Break, time=0.2, value=1.2}
          {arriveTangent=-2.7631, interpMode=RCIM_Cubic, leaveTangent=-2.7631, tangentMode=RCTM_User, time=1, value=0}
         ]}}}
      Scale Curve=&Multiply_Float(A=0.4, B=&FN_FrontEndFloatModifier(Gameplay=1, Front End=$System.FrontendScale))
     ))) {module=ScaleSpriteSize001}
   &Drag(Drag=5, Use Linear Drag=true, Use Rotational Drag=false)
   &GravityForce(Gravity=<0 0 -100>)
   &CurlNoiseForce(Bias Noise=false, Noise Strength=200, Pan Noise=false)
   &InheritSourceMovement(ApplyForce=false, ApplyPosition=true, ApplyPositionFalloff=false)
   &SolveForcesAndVelocity(Clamp Velocity=false, Limit Acceleration=false)
   &ScaleSpriteSizeBySpeed(Max Scale Factor=<0.8 2>, Velocity Threshold=100)
   /SetVariables(Particles.Glow03_Color=&MakeLinearColorFromVectorAndFloat(
     "Vector (RGB)"=&Multiply_Vector(
      A=&MakeVectorFromLinearColorRGB(Color=$Particles.Color)
      B=&MakeVectorFromLinearColorRGB(Color=<rgba 1 1 1 1>)
     )
     "Float (Alpha)"=1
    )) {isSetParameters=true}
   /SetVariables(Particles.Glow03_Size=&Multiply_Vector2D(A=$Particles.SpriteSize, B=<6 2>)) {isSetParameters=true}
  ]
  rendererClasses [&NiagaraSpriteRendererProperties]
  renderers [
   {rendererClass=&NiagaraSpriteRendererProperties}
   {rendererClass=&NiagaraSpriteRendererProperties, rendererIndex=1}
  ]
 }
 Squares {
  emitterUpdateScript [
   &EmitterState()
   &SpawnBurst_Instantaneous(
    Spawn Count=&RandomRangeInt(Minimum=26, Maximum=30)
    Use Spawn Probability=false
    UseLoopCountLimit=false
   )
  ]
  particleSpawnScript [
   &InitializeParticle(
    Color=&SelectLinearColorFromArray(
     Array Sampling Mode=::NewEnumerator0
     Direct Array Index=?
     Color Selection Array=DataInterface() {propertyValues {ColorData [<rgba 1.5 0.075 0 1>, <rgba 1 0 0 1>, <rgba 0.3073 0.0102 0 1>]}}
     Random Seed=0
    )
    Lifetime Max=0.6
    Lifetime Min=0.25
    Lifetime Mode=::NewEnumerator1
    Sprite Rotation Angle=&SelectFloatFromArray(
     Array Sampling Mode=::NewEnumerator0
     Direct Array Index=?
     Float Selection Array=DataInterface() {propertyValues {FloatData [0,0.25,0.5,0.75]}}
     Random Seed=0
    )
    Sprite Rotation Mode=::NewEnumerator1
    Sprite Size Mode=::NewEnumerator1
    UsePositionOffset=false
    Write Material Random=true
   )
   &SocketLocation(
    Interpolate Socket Sampling=true
    SamplePosOffset=true
    Socket Reader=$System.Socket Reader
    Socket Sampling Mode=::NewEnumerator3
    VelocitySpeedLimit=false
   )
   &ShapeLocation(
    MaskSpawnGroup=false
    Override Seed=false
    Sphere Radius=&Multiply_Float(A=40, B=&FN_FrontEndFloatModifier(Gameplay=1, Front End=$System.FrontendScale))
    Sphere Surface Distribution=0.6071
    UseOwnerScale=true
   )
   &AddVelocity(
    Module.Use Falloff Distance=false
    Override Seed=false
    Velocity Mode=::NewEnumerator1
    Velocity Origin=$Output.SocketLocation.SampledPosition
    Velocity Speed=&Multiply_Float(
     A=&RandomRangeFloat(Minimum=300, Maximum=600)
     B=&FN_FrontEndFloatModifier(Gameplay=1, Front End=$System.FrontendScale)
    )
   )
   &CameraOffset(Camera Offset Amount=15)
  ]
  particleUpdateScript [
   &ParticleState()
   &Drag(Drag=&RandomRangeFloat(Minimum=5, Maximum=8), Use Linear Drag=true, Use Rotational Drag=false)
   &ScaleSpriteSize(
    Uniform Curve Scale=&FN_FrontEndFloatModifier(Gameplay=1, Front End=$System.FrontendScale)
    Uniform Curve Sprite Scale=DataInterface() {propertyValues {Curve {
       defaultValue=0
       keys [
        {arriveTangent=0.9996, interpMode=RCIM_Constant, time=0, value=0.8}
        <at 0.3 1 RCIM_Constant>
        <at 0.5 0.75 RCIM_Constant>
        {interpMode=RCIM_Cubic, tangentMode=RCTM_User, time=0.7, value=1}
        <at 1 0 RCIM_Cubic>
       ]
      }}}
   )
   &ScaleColor(
    Scale Alpha=&FloatFromCurve(FloatCurve=DataInterface() {propertyValues {Curve {
        defaultValue=0
        keys [{interpMode=RCIM_Cubic, tangentMode=RCTM_User, time=0, value=1}, {interpMode=RCIM_Cubic, tangentMode=RCTM_User, time=1, value=0}]
       }}})
    ScaleA=true
    ScaleRGB=true
   ) {module=ScaleColor001}
   &CurlNoiseForce(Bias Noise=false, Noise Strength=100, Pan Noise=false)
   &InheritSourceMovement(ApplyForce=false, ApplyPosition=true, ApplyPositionFalloff=false)
   &SolveForcesAndVelocity(Clamp Velocity=false, Limit Acceleration=false)
   &FN_Sprite_Attributes(
    Color=&MakeLinearColorFromVectorAndFloat(
     "Vector (RGB)"=&MakeVectorFromLinearColorRGB(Color=<rgba 0.026 0 0 1>)
     "Float (Alpha)"=&FloatFromCurve(
      FloatCurve=DataInterface() {propertyValues {Curve {
         defaultValue=0
         keys [{interpMode=RCIM_Cubic, tangentMode=RCTM_User, time=0, value=1}, {interpMode=RCIM_Cubic, tangentMode=RCTM_User, time=1, value=0}]
        }}}
      Scale Curve=10
     )
    )
    Create Color Attribute=true
    Create SpriteSize Attribute=true
    SpriteSize=&Multiply_Vector2D_ByFloat(Float=1.25, Vector2D=$Particles.SpriteSize)
    Write to Module Namespace=true
   )
  ]
  rendererClasses [&NiagaraSpriteRendererProperties]
  renderers [
   {rendererClass=&NiagaraSpriteRendererProperties}
   {rendererClass=&NiagaraSpriteRendererProperties, rendererIndex=1}
  ]
 }
}
summary {
 systemName=NS_Emote_VileSting_Landing
 emitters [
  {emitterName=Squares, rendererClasses [&NiagaraSpriteRendererProperties]}
  {emitterName=Airflow_Spline, rendererClasses [&NiagaraMeshRendererProperties]}
  {emitterName=Airflow_Radial, rendererClasses [&NiagaraMeshRendererProperties]}
  {emitterName=Sparks, rendererClasses [&NiagaraSpriteRendererProperties]}
  {emitterName=EnergyWave, rendererClasses [&NiagaraMeshRendererProperties]}
  {emitterName=Glows, rendererClasses [&NiagaraSpriteRendererProperties, &NiagaraLightRendererProperties]}
  {emitterName=Notes_Line, rendererClasses [&NiagaraRibbonRendererProperties]}
  {emitterName=Block_Ring, rendererClasses [&NiagaraSpriteRendererProperties]}
 ]
 userVariables [
  User.bIsFrontEnd {type=NiagaraBool}
  User.bIsFrontEndPreview {type=NiagaraBool}
 ]
}
